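# Translucent, unlit, two-sided Niagara mesh-particle material, a glow effect for the Cube Ninja Male pickaxe.
# Panning, desaturatable smoke texture plus Fresnel is tinted by emissive and particle colors, then passed through an emissive-adjustment function.
# Opacity multiplies particle alpha, Fresnel, a UV1 vertical gradient mask and a panning noise dissolve driven by a dynamic parameter.
# The Fresnel parameter's Exp channel feeds both exponent and base reflect fraction, leaving Base unused.
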Material /BRCosmetics/Weapons/FORT_Melee/Pickaxe_Cube_Ninja_Male/FX/M_Pickaxe_Cube_Ninja_Male_Mesh_Glow
graph {
 properties {
  bUseTranslucencyVertexFog=false
  bUsedWithNiagaraMeshParticles=true
  blendMode=BLEND_Translucent
  pixelDepthOffsetMode=PDOM_Legacy
  shadingModel=MSM_Unlit
  twoSided=true
  usedShadingModels=Unlit
 }
 nodes [
  n1 = VectorParameter() {group=Emissive, parameterName=Emissive Color, defaultValue=<rgba 1 1 1 1>}
  n2 = TextureCoordinate() {coordinateIndex=1, uTiling=0.5, vTiling=0.5}
  n3 = VectorParameter() {group=Emissive, parameterName=Emissive UV, defaultValue=<rgba 1 1 0 0>}
  n4 = AppendVector(A=@n3.R, B=@n3.R)
  n5 = AppendVector(A=@n3.B, B=@n3.B)
  n6 = Time()
  n7 = TextureSampleParameter2D(UVs=Add(A=Multiply(A=@n2, B=@n4), B=Multiply(A=@n5, B=@n6))) {group=Emissive, parameterName=Emissive Tex, texture=&T_Inky_Smoke_Tile_Chromatic}
  n8 = ScalarParameter() {group=Emissive, parameterName=Desaturation}
  n9 = VectorParameter() {group=Fresnel, parameterName=Fresnel, channelNames {a="", b=Intensity, g=Base, r=Exp}, defaultValue=<rgba 3 0 1 0>}
  n10 = Multiply(A=Fresnel(ExponentIn=@n9.Exp, BaseReflectFractionIn=@n9.Exp) {baseReflectFraction=0.02, exponent=3}, B=@n9.Intensity) {constA=1}
  n11 = Add(
   A=Multiply(A=Desaturation(None=@n7.RGB, Fraction=@n8), B=ScalarParameter() {defaultValue=1, group=Emissive, parameterName=Emissive Intersity}) {constB=0.1}
   B=@n10
  )
  n12 = ParticleColor()
  MaterialFunctionCall(Emissive=Multiply(A=Multiply(A=@n1.RGB, B=@n11), B=@n12.RGB)) {materialFunction=&MF_FX_EmissiveAdjustments}
  n13 = OneMinus(None=ComponentMask(None=TextureCoordinate() {coordinateIndex=1}) {g=true})
  n14 = VectorParameter() {parameterName=Mask, channelNames {a="", b="", g=Multiply, r=Power}, defaultValue=<rgba 1 1 0 0>}
  n15 = Saturate(None=Multiply(A=Power(Base=@n13, Exp=@n14.Power) {constExponent=4}, B=@n14.Multiply) {constB=5})
  n16 = TextureCoordinate() {coordinateIndex=1, uTiling=0.5, vTiling=0.5}
  n17 = VectorParameter() {group=Dissolve, parameterName=Dissolve UV, defaultValue=<rgba 1 1 0 0>}
  n18 = AppendVector(A=@n17.R, B=@n17.R)
  n19 = AppendVector(A=@n17.B, B=@n17.B)
  n20 = Time()
  n21 = TextureSampleParameter2D(UVs=Add(A=Multiply(A=@n16, B=@n18), B=Multiply(A=@n19, B=@n20))) {
   group=Dissolve
   parameterName=Dissolve Tex
   samplerType=SAMPLERTYPE_Masks
   texture=/Game/Characters/Player/Male/Medium/Bodies/M_MED_Storm_Soldier/Textures/ElectricityNoise_M
  }
  n22 = ScalarParameter() {defaultValue=3, group=Dissolve, parameterName=Dissolve Smooth}
  n23 = DynamicParameter() {paramNames [Dissolve,Param2,Param3,Param4]}
  n24 = ScalarParameter() {defaultValue=1, group=Dissolve, parameterName=Dissolve}
  n25 = Subtract(A=Multiply(A=@n21.R, B=@n22), B=LinearInterpolate(A=@n22, Alpha=Multiply(A=@n23.Dissolve, B=@n24)) {constB=-1})
  Multiply(A=@n12.A, B=Multiply(A=Multiply(A=@n10, B=@n15), B=Saturate(None=@n25)))
 ]
}
